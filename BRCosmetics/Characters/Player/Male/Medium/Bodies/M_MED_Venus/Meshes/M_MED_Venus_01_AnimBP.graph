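# Animation Blueprint for the M_MED_Venus_01 cosmetic body that adds secondary motion to its calf, shoulder and neck tendrils.
# On initialization it caches the owner's skeletal mesh component, and the AnimGraph copies that mesh's pose as its base.
# A serial chain of single-bone Anim Dynamics nodes then layers spring-driven, limit-constrained motion in actor space, blended at low alpha.
# The dynamics nodes carry an LOD threshold, so they run only at the highest-detail LODs.
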AnimBlueprint /BRCosmetics/Characters/Player/Male/Medium/Bodies/M_MED_Venus/Meshes/M_MED_Venus_01_AnimBP
target_skeleton=&/M_MED_Venus_01_Skeleton
graphs [
 AnimGraph {nodes [
   n1 = Get MeshToCopy() -> (MeshToCopy:Skeletal Mesh Component Object Reference)
   n2 = Anim Dynamics - Bone: dyn_calfTendril_01_l(ComponentPose=Local To Component(LocalPose=Copy Pose From Mesh(SourceMeshComponent=@n1)), Alpha=0.2, ExternalForce="0,0,0") {
    angularSpringConstant=80
    bAngularSpring=true
    bGravityOverrideInSimSpace=true
    bLinearSpring=true
    lODThreshold=1
    linearSpringConstant=80
    simulationSpace=Actor
    alphaBoolBlend {blendOption=Linear}
    boundBone {boneName=dyn_calfTendril_01_l}
    physicsBodyDefinitions [{
      sphereCollisionRadius=0
      boundBone {boneName=dyn_calfTendril_01_l}
      boxExtents=<30 30 30>
      constraintSetup {
       angularTargetAxis=AxisY
       angularLimitsMax=<2.5 50 50>
       angularLimitsMin=<-2.5 -50 -50>
       angularTarget=<0 0 0>
       linearAxesMax=<10 20 20>
       linearAxesMin=<0 -20 -20>
      }
      localJointOffset=<-3 0 0>
     }]
   }
   n3 = Anim Dynamics - Bone: dyn_calfTendril_02_l(ComponentPose=@n2, Alpha=0.15, ExternalForce="0,0,0") {
    angularSpringConstant=80
    bAngularSpring=true
    bGravityOverrideInSimSpace=true
    bLinearSpring=true
    lODThreshold=1
    linearSpringConstant=80
    simulationSpace=Actor
    alphaBoolBlend {blendOption=Linear}
    boundBone {boneName=dyn_calfTendril_02_l}
    physicsBodyDefinitions [{
      sphereCollisionRadius=0
      boundBone {boneName=dyn_calfTendril_02_l}
      boxExtents=<30 30 30>
      constraintSetup {
       angularTargetAxis=AxisY
       angularLimitsMax=<2.5 30 40>
       angularLimitsMin=<-2.5 -40 -30>
       angularTarget=<0 0 0>
       linearAxesMax=<10 20 20>
       linearAxesMin=<0 -20 -20>
      }
      localJointOffset=<3 0 0>
     }]
   }
   n4 = Anim Dynamics - Bone: dyn_calfTendril_03_l(ComponentPose=@n3, Alpha=0.15, ExternalForce="0,0,0") {
    angularSpringConstant=80
    bAngularSpring=true
    bGravityOverrideInSimSpace=true
    bLinearSpring=true
    lODThreshold=1
    linearSpringConstant=80
    simulationSpace=Actor
    alphaBoolBlend {blendOption=Linear}
    boundBone {boneName=dyn_calfTendril_03_l}
    physicsBodyDefinitions [{
      sphereCollisionRadius=0
      boundBone {boneName=dyn_calfTendril_03_l}
      boxExtents=<30 30 30>
      constraintSetup {
       angularTargetAxis=AxisY
       angularLimitsMax=<2.5 30 45>
       angularLimitsMin=<-2.5 -35 -30>
       angularTarget=<0 0 0>
       linearAxesMax=<10 20 20>
       linearAxesMin=<0 -20 -20>
      }
      localJointOffset=<0 -3 0>
     }]
   }
   n5 = Anim Dynamics - Bone: dyn_calfTendril_02_r(ComponentPose=@n4, Alpha=0.2, ExternalForce="0,0,0") {
    angularSpringConstant=80
    bAngularSpring=true
    bGravityOverrideInSimSpace=true
    bLinearSpring=true
    lODThreshold=1
    linearSpringConstant=80
    simulationSpace=Actor
    alphaBoolBlend {blendOption=Linear}
    boundBone {boneName=dyn_calfTendril_02_r}
    physicsBodyDefinitions [{
      sphereCollisionRadius=0
      boundBone {boneName=dyn_calfTendril_02_r}
      boxExtents=<30 30 30>
      constraintSetup {
       angularTargetAxis=AxisY
       angularLimitsMax=<2.5 50 50>
       angularLimitsMin=<-2.5 -50 -50>
       angularTarget=<0 0 0>
       linearAxesMax=<10 20 20>
       linearAxesMin=<0 -20 -20>
      }
      localJointOffset=<3 0 0>
     }]
   }
   n6 = Anim Dynamics - Bone: dyn_calfTendril_01_r(ComponentPose=@n5, Alpha=0.15, ExternalForce="0,0,0") {
    angularSpringConstant=80
    bAngularSpring=true
    bGravityOverrideInSimSpace=true
    bLinearSpring=true
    lODThreshold=1
    linearSpringConstant=80
    simulationSpace=Actor
    alphaBoolBlend {blendOption=Linear}
    boundBone {boneName=dyn_calfTendril_01_r}
    physicsBodyDefinitions [{
      sphereCollisionRadius=0
      boundBone {boneName=dyn_calfTendril_01_r}
      boxExtents=<30 30 30>
      constraintSetup {
       angularTargetAxis=AxisY
       angularLimitsMax=<2.5 30 40>
       angularLimitsMin=<-2.5 -40 -30>
       angularTarget=<0 0 0>
       linearAxesMax=<10 20 20>
       linearAxesMin=<0 -20 -20>
      }
      localJointOffset=<0 -3 0>
     }]
   }
   n7 = Anim Dynamics - Bone: dyn_calfTendril_03_r(ComponentPose=@n6, Alpha=0.15, ExternalForce="0,0,0") {
    angularSpringConstant=80
    bAngularSpring=true
    bGravityOverrideInSimSpace=true
    bLinearSpring=true
    lODThreshold=1
    linearSpringConstant=80
    simulationSpace=Actor
    alphaBoolBlend {blendOption=Linear}
    boundBone {boneName=dyn_calfTendril_03_r}
    physicsBodyDefinitions [{
      sphereCollisionRadius=0
      boundBone {boneName=dyn_calfTendril_03_r}
      boxExtents=<30 30 30>
      constraintSetup {
       angularTargetAxis=AxisY
       angularLimitsMax=<2.5 30 45>
       angularLimitsMin=<-2.5 -35 -30>
       angularTarget=<0 0 0>
       linearAxesMax=<10 20 20>
       linearAxesMin=<0 -20 -20>
      }
      localJointOffset=<-3 0 0>
     }]
   }
   n8 = Anim Dynamics - Bone: dyn_shoulderTendril_01_l(ComponentPose=@n7, Alpha=0.2, ExternalForce="0,0,0") {
    angularSpringConstant=80
    bAngularSpring=true
    bGravityOverrideInSimSpace=true
    bLinearSpring=true
    gravityScale=0.5
    lODThreshold=1
    linearSpringConstant=80
    simulationSpace=Actor
    alphaBoolBlend {blendOption=Linear}
    boundBone {boneName=dyn_shoulderTendril_01_l}
    physicsBodyDefinitions [{
      sphereCollisionRadius=0
      boundBone {boneName=dyn_shoulderTendril_01_l}
      boxExtents=<30 30 30>
      constraintSetup {
       angularTargetAxis=AxisY
       angularLimitsMax=<2.5 50 50>
       angularLimitsMin=<-2.5 -50 -50>
       angularTarget=<0 0 0>
       linearAxesMax=<0 5 5>
       linearAxesMin=<0 -5 -5>
      }
      localJointOffset=<4.5 0 0>
     }]
   }
   n9 = Anim Dynamics - Bone: dyn_shoulderTendril_02_l(ComponentPose=@n8, Alpha=0.2, ExternalForce="0,0,0") {
    angularSpringConstant=80
    bAngularSpring=true
    bGravityOverrideInSimSpace=true
    bLinearSpring=true
    gravityScale=0.5
    lODThreshold=1
    linearSpringConstant=80
    simulationSpace=Actor
    alphaBoolBlend {blendOption=Linear}
    boundBone {boneName=dyn_shoulderTendril_02_l}
    physicsBodyDefinitions [{
      sphereCollisionRadius=0
      boundBone {boneName=dyn_shoulderTendril_02_l}
      boxExtents=<30 30 30>
      constraintSetup {
       angularTargetAxis=AxisY
       angularLimitsMax=<2.5 30 40>
       angularLimitsMin=<-2.5 -40 -30>
       angularTarget=<0 0 0>
       linearAxesMax=<0 5 5>
       linearAxesMin=<0 -5 -5>
      }
      localJointOffset=<4.5 0 0>
     }]
   }
   n10 = Anim Dynamics - Bone: dyn_shoulderTendril_03_l(ComponentPose=@n9, Alpha=0.2, ExternalForce="0,0,0") {
    angularSpringConstant=80
    bAngularSpring=true
    bGravityOverrideInSimSpace=true
    bLinearSpring=true
    gravityScale=0.5
    lODThreshold=1
    linearSpringConstant=80
    simulationSpace=Actor
    alphaBoolBlend {blendOption=Linear}
    boundBone {boneName=dyn_shoulderTendril_03_l}
    physicsBodyDefinitions [{
      sphereCollisionRadius=0
      boundBone {boneName=dyn_shoulderTendril_03_l}
      boxExtents=<30 30 30>
      constraintSetup {
       angularTargetAxis=AxisY
       angularLimitsMax=<2.5 30 45>
       angularLimitsMin=<-2.5 -35 -30>
       angularTarget=<0 0 0>
       linearAxesMax=<0 5 5>
       linearAxesMin=<0 -5 -5>
      }
      localJointOffset=<4.5 0 0>
     }]
   }
   n11 = Anim Dynamics - Bone: dyn_shoulderTendril_01_r(ComponentPose=@n10, Alpha=0.2, ExternalForce="0,0,0") {
    angularSpringConstant=80
    bAngularSpring=true
    bGravityOverrideInSimSpace=true
    bLinearSpring=true
    gravityScale=0.75
    lODThreshold=1
    linearSpringConstant=80
    simulationSpace=Actor
    alphaBoolBlend {blendOption=Linear}
    boundBone {boneName=dyn_shoulderTendril_01_r}
    physicsBodyDefinitions [{
      sphereCollisionRadius=0
      boundBone {boneName=dyn_shoulderTendril_01_r}
      boxExtents=<30 30 30>
      constraintSetup {
       angularTargetAxis=AxisY
       angularLimitsMax=<2.5 50 50>
       angularLimitsMin=<-2.5 -50 -50>
       angularTarget=<0 0 0>
       linearAxesMax=<0 5 5>
       linearAxesMin=<0 -5 -5>
      }
      localJointOffset=<-4.5 0 0>
     }]
   }
   n12 = Anim Dynamics - Bone: dyn_shoulderTendril_02_r(ComponentPose=@n11, Alpha=0.2, ExternalForce="0,0,0") {
    angularSpringConstant=80
    bAngularSpring=true
    bGravityOverrideInSimSpace=true
    bLinearSpring=true
    gravityScale=0.75
    lODThreshold=1
    linearSpringConstant=80
    simulationSpace=Actor
    alphaBoolBlend {blendOption=Linear}
    boundBone {boneName=dyn_shoulderTendril_02_r}
    physicsBodyDefinitions [{
      sphereCollisionRadius=0
      boundBone {boneName=dyn_shoulderTendril_02_r}
      boxExtents=<30 30 30>
      constraintSetup {
       angularTargetAxis=AxisY
       angularLimitsMax=<2.5 30 40>
       angularLimitsMin=<-2.5 -40 -30>
       angularTarget=<0 0 0>
       linearAxesMax=<0 5 5>
       linearAxesMin=<0 -5 -5>
      }
      localJointOffset=<-4.5 0 0>
     }]
   }
   n13 = Anim Dynamics - Bone: dyn_shoulderTendril_03_r(ComponentPose=@n12, Alpha=0.2, ExternalForce="0,0,0") {
    angularSpringConstant=80
    bAngularSpring=true
    bGravityOverrideInSimSpace=true
    bLinearSpring=true
    gravityScale=0.75
    lODThreshold=1
    linearSpringConstant=80
    simulationSpace=Actor
    alphaBoolBlend {blendOption=Linear}
    boundBone {boneName=dyn_shoulderTendril_03_r}
    physicsBodyDefinitions [{
      sphereCollisionRadius=0
      boundBone {boneName=dyn_shoulderTendril_03_r}
      boxExtents=<30 30 30>
      constraintSetup {
       angularTargetAxis=AxisY
       angularLimitsMax=<2.5 30 45>
       angularLimitsMin=<-2.5 -35 -30>
       angularTarget=<0 0 0>
       linearAxesMax=<0 5 5>
       linearAxesMin=<0 -5 -5>
      }
      localJointOffset=<-4.5 0 0>
     }]
   }
   n14 = Anim Dynamics - Bone: dyn_neckTendril_b(ComponentPose=@n13, Alpha=0.1, ExternalForce="0,0,0") {
    angularSpringConstant=80
    bAngularSpring=true
    bGravityOverrideInSimSpace=true
    bLinearSpring=true
    gravityScale=0.75
    lODThreshold=1
    linearSpringConstant=80
    simulationSpace=Actor
    alphaBoolBlend {blendOption=Linear}
    boundBone {boneName=dyn_neckTendril_b}
    physicsBodyDefinitions [{
      sphereCollisionRadius=0
      boundBone {boneName=dyn_neckTendril_b}
      boxExtents=<30 30 30>
      constraintSetup {
       angularTargetAxis=AxisY
       angularLimitsMax=<2.5 50 50>
       angularLimitsMin=<-2.5 -50 -50>
       angularTarget=<0 0 0>
       linearAxesMax=<0 5 5>
       linearAxesMin=<0 -5 -5>
      }
      localJointOffset=<0 -3 0>
     }]
   }
   n15 = Anim Dynamics - Bone: dyn_neckTendril_l(ComponentPose=@n14, Alpha=0.1, ExternalForce="0,0,0") {
    angularSpringConstant=80
    bAngularSpring=true
    bGravityOverrideInSimSpace=true
    bLinearSpring=true
    gravityScale=0.75
    lODThreshold=1
    linearSpringConstant=80
    simulationSpace=Actor
    alphaBoolBlend {blendOption=Linear}
    boundBone {boneName=dyn_neckTendril_l}
    physicsBodyDefinitions [{
      sphereCollisionRadius=0
      boundBone {boneName=dyn_neckTendril_l}
      boxExtents=<30 30 30>
      constraintSetup {
       angularTargetAxis=AxisY
       angularLimitsMax=<2.5 30 40>
       angularLimitsMin=<-2.5 -40 -30>
       angularTarget=<0 0 0>
       linearAxesMax=<0 5 5>
       linearAxesMin=<0 -5 -5>
      }
      localJointOffset=<0 0 0>
     }]
   }
   n16 = Anim Dynamics - Bone: dyn_neckTendril_r(ComponentPose=@n15, Alpha=0.1, ExternalForce="0,0,0") {
    angularSpringConstant=80
    bAngularSpring=true
    bGravityOverrideInSimSpace=true
    bLinearSpring=true
    gravityScale=0.75
    lODThreshold=1
    linearSpringConstant=80
    simulationSpace=Actor
    alphaBoolBlend {blendOption=Linear}
    boundBone {boneName=dyn_neckTendril_r}
    physicsBodyDefinitions [{
      sphereCollisionRadius=0
      boundBone {boneName=dyn_neckTendril_r}
      boxExtents=<30 30 30>
      constraintSetup {
       angularTargetAxis=AxisY
       angularLimitsMax=<2.5 30 45>
       angularLimitsMin=<-2.5 -35 -30>
       angularTarget=<0 0 0>
       linearAxesMax=<0 5 5>
       linearAxesMin=<0 -5 -5>
      }
      localJointOffset=<0 0 0>
     }]
   }
   Output Pose(Result=Component To Local(ComponentPose=@n16)) {layerGroup=None}
  ]}
 EventGraph {nodes [
   Event BlueprintInitializeAnimation()
   > Set MeshToCopy(MeshToCopy:Skeletal Mesh Component Object Reference=GetOwnerSkeletalMeshComponent()) -> (Output_Get:Skeletal Mesh Component Object Reference)
   Event BlueprintUpdateAnimation()
  ]}
]
variables [MeshToCopy {type=SkeletalMeshComponent}]
